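# Material function for the PolarBlind Vein quadruped companion that adds an animated, curve-remapped color glow to incoming material attributes.
# An inverted Fresnel or gradient texture channel, offset by panning world-aligned noise on pre-skinned inputs and blended by a mask, drives a curve-atlas color lookup, optionally cycled over time.
# The color is masked by an FX channel, scaled by a time-of-day curve and added to emissive, while roughness gets a minimum clamp.
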MaterialFunction /CosmeticCompanions/Assets/Quadruped/PolarBlind_Vein/Material/MF_RemapCurveByFresnel_PolarBlind_Vein
graph {nodes [
  n1 = NamedRerouteUsage()
  n2 = ScalarParameter() {defaultValue=1.8, group=RemapCurveByFresnel, parameterName=RemapCurveFresnelExponent}
  n3 = GetMaterialAttributes(None=FunctionInput() {inputName=MA, inputType=FunctionInput_MaterialAttributes}) {attributeGetTypes [#1, #2, #3]}
  n4 = Transform(None=@n3.Normal)
  n5 = StaticSwitchParameter(
   True=ChannelMaskParameter(None=TextureSampleParameter2D() {group=RemapCurveByFresnel, parameterName=TextureGradient, texture=&T_BlacksRGB}.RGB) {group=RemapCurveByFresnel, parameterName=TextureGradientChannelMask}.R
   False=OneMinus(None=MaterialFunctionCall(Exponent=@n2, Vertex / Pixel Normal=@n4) {materialFunction=&MF_Fresnel_FX}.Fresnel)
  ) {group=RemapCurveByFresnel, parameterName=UseTextureGradient}
  n6 = TextureObjectParameter() {
   group=RemapCurveByFresnel
   parameterName=NoiseTexture
   samplerType=SAMPLERTYPE_LinearColor
   texture=/Game/Weapons/WeaponSkins/Textures/T_Swirly_Noise_highContrast1
  }
  n7 = FunctionInput(Preview=WorldPosition().XYZ) {inputName=PreskinLocalPosition, sortPriority=3}
  n8 = Multiply(A=Time(), B=VectorParameter() {group=RemapCurveByFresnel, parameterName=NoisePanSpeed, defaultValue=<rgba 0.66 1 2 0>}.RGB)
  n9 = PreSkinnedNormal()
  n10 = ComponentMask(None=MaterialFunctionCall(
    TextureObject=@n6
    TextureSize=VectorParameter() {group=RemapCurveByFresnel, parameterName=NoiseWorldTileScale, defaultValue=<rgba 105 105 105 0>}.RGB
    WorldPosition=Subtract(A=@n7, B=@n8)
    World Space Normal=VertexInterpolator(VS=@n9).PS
   ) {materialFunction=/Game/Weapons/WeaponSkins/Wraps/MaterialFunctions/WorldAlignedTexture}.XYZ Texture) {r=true}
  n11 = LinearInterpolate(Alpha=@n10) {constA=0.75}
  n12 = Multiply(A=@n11) {constB=0.1}
  n13 = Multiply(
   A=Subtract(A=@n10) {constB=0.5}
   B=LinearInterpolate(
    A=ScalarParameter() {defaultValue=0.5, group=RemapCurveByFresnel, parameterName=NoiseScaleOuter}
    B=ScalarParameter() {defaultValue=0.1, group=RemapCurveByFresnel, parameterName=NoiseScaleInner}
    Alpha=@n5
   )
  )
  n14 = TextureSample() {
   samplerSource=SSM_Wrap_WorldGroupSettings
   samplerType=SAMPLERTYPE_Masks
   texture=/CosmeticCompanions/Assets/Quadruped/PolarBlind_Vein/Texture/T_PolarBlind_TAMask
  }
  n15 = Add(A=@n5, B=LinearInterpolate(A=Add(A=@n12, B=@n13), B=@n11, Alpha=@n14.B))
  n16 = Add(
   A=MaterialFunctionCall!(Time, Divisor=ScalarParameter() {defaultValue=2, group=RemapCurveByFresnel, parameterName=AddTimeToCurveLookupDivisor}, Mod) {materialFunction=/Game/Effects/Fort_Effects/Materials/_Functions/MF_ModuloTime}.Clamped 0-1
   B=@n15
  )
  n17 = Constant() {r=1}
  n18 = CurveAtlasRowParameter(CurveTime=StaticSwitchParameter(True=Fmod(A=@n16, B=@n17), False=@n15) {group=RemapCurveByFresnel, parameterName=AddTimeToCurveLookup}) {
   atlas=/BRCosmetics/Characters/Player/Male/Medium/Bodies/M_MED_Keychain/Textures/A_KeyChainAtlas
   curve=/BRCosmetics/Characters/Player/Male/Medium/Bodies/M_MED_Keychain/Textures/C_KeyChainColorInner
   group=RemapCurveByFresnel
   parameterName=RemapColorCurveKeyChain
  }
  n19 = TextureSample(Tex=FunctionInput(Preview=TextureObject()) {bUsePreviewValueAsDefault=true, inputName=FXMask, inputType=FunctionInput_Texture2D, sortPriority=1}) {samplerSource=SSM_Wrap_WorldGroupSettings}
  n20 = CurveAtlasRowParameter(CurveTime=CollectionParameter() {collection=&FortniteMaterialParameters, parameterName=TimeOfDay_0-1}) {
   atlas=&AthenaCurveAtlas01
   curve=/Game/Characters/Player/Male/Medium/Bodies/M_MED_Vivid_Color/Materials/C_Med_Vivid_Color_TOD_Values
   defaultValue=295
   group=RemapCurveByFresnel
   parameterName=TOD_Curve
  }
  n21 = Multiply(A=Multiply(A=@n18, B=ChannelMaskParameter(None=@n19.RGBA) {group=RemapCurveByFresnel, parameterName=FXMask_ChannelPicker}.R), B=@n20)
  n22 = NamedRerouteUsage()
  FunctionOutput(None=SetMaterialAttributes!(MaterialAttributes=NamedRerouteUsage(), EmissiveColor=Add(A=@n1, B=@n21), Roughness=Clamp(None=@n22) {minDefault=0.2}) {attributeSetTypes [#1, #3]})
  NamedRerouteDeclaration(None=@n3.MaterialAttributes) {name=MA_IN, nodeColor=<rgba 0.8281 0 1 1>}
  NamedRerouteDeclaration(None=@n3.EmissiveColor) {name=EmissiveColor_In, nodeColor=<rgba 1 0 0.0703 1>}
  NamedRerouteDeclaration(None=@n3.Roughness) {name=Roughness_In, nodeColor=<rgba 0.1016 0 1 1>}
 ]}
