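# Parent skeletal-mesh material for the PolarBlind Vein quadruped cosmetic companion, layering a stylized emissive effect between the shared character PreFX and PostFX functions.
# A fresnel-based curve remap feeds emissive, which is scaled by object view depth, desaturated and passed through StuffRasp's reactive function.
# A tinted head fresnel is added, then masked and colored by X_Color and TongueColor before overriding emissive and specular.
# Packed mask textures supply skeleton AO, head and tongue masks through named reroutes.
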Material /CosmeticCompanions/Assets/Quadruped/PolarBlind_Vein/Material/M_PolarBlind_Vein_X_Parent
graph {
 properties {
  bUseMaterialAttributes=true
  bUsedWithMeshDeformer=true
  bUsedWithMorphTargets=true
  bUsedWithSkeletalMesh=true
  pixelDepthOffsetMode=PDOM_Legacy
  shadingModel=MSM_FromMaterialExpression
  usedShadingModels=Default Lit | Subsurface | Clear Coat | Subsurface Profile | Cloth
 }
 nodes [
  Constant() {r=1}
  n1 = ScalarParameter() {defaultValue=0.3029, group=StuffRasp, parameterName=HeadFresnel_Exponentln}
  n2 = VertexNormalWS()
  n3 = Power(Base=Multiply(A=Fresnel(ExponentIn=@n1, Normal=@n2), B=Constant() {r=2}), Exp=Constant() {r=4.821})
  n4 = NamedRerouteUsage()
  n5 = MaterialFunctionCall() {materialFunction=/BRCosmetics/Characters/Player/Male/Medium/Bodies/M_MED_StuffRasp_Maze/Material/MF_ObjectViewDepth}
  n6 = MaterialFunctionCall() {materialFunction=&MF_CharacterParent_2023_PreFX}
  n7 = TextureObjectParameter() {
   group=RemapCurveByFresnel
   parameterName=TechArtMaskRe
   samplerType=SAMPLERTYPE_Masks
   texture=&T_WhiteMask
  }
  n8 = GetMaterialAttributes(None=MaterialFunctionCall!(MA=@n6.Result, FXMask=@n7, PreskinLocalPosition=@n6.Result) {materialFunction=&/MF_RemapCurveByFresnel_PolarBlind_Vein}.Result) {attributeGetTypes [#1, #2]}
  n9 = Multiply(A=@n8.Emissive Color, B=LinearInterpolate(A=Constant(), B=Constant() {r=3.85}, Alpha=Saturate(None=@n5.Result)))
  n10 = Constant() {r=1}
  n11 = Add(
   A=Multiply(A=Multiply(A=@n3, B=@n4), B=VectorParameter() {group=StuffRasp, parameterName=HeadFresnelColor, defaultValue=<rgba 1 1 1 1>}.RGB)
   B=MaterialFunctionCall!(In=Desaturation(None=@n9, Fraction=@n10)) {materialFunction=/BRCosmetics/Characters/Player/Male/Medium/Bodies/M_MED_StuffRasp_Maze/Material/MF_StuffRasp_Reactive}.OutReactive
  )
  n12 = LinearInterpolate(A=NamedRerouteUsage(), Alpha=NamedRerouteUsage())
  n13 = Power(Base=NamedRerouteUsage())
  n14 = NamedRerouteUsage()
  n15 = NamedRerouteUsage()
  n16 = Multiply(
   A=Multiply(A=Multiply(A=@n11, B=@n12), B=LinearInterpolate(B=@n13, Alpha=@n14) {constA=1})
   B=LinearInterpolate(A=VectorParameter() {group=StuffRasp, parameterName=X_Color, defaultValue=<rgba 0.9 0.9501 1 1>}.RGB, Alpha=@n15) {constB=0}
  )
  n17 = VectorParameter() {group=StuffRasp, parameterName=TongueColor, defaultValue=<rgba 0 0 0 1>}
  n18 = NamedRerouteUsage()
  n19 = Constant()
  n20 = NamedRerouteUsage()
  MaterialFunctionCall(
   Material=SetMaterialAttributes!(
    MaterialAttributes=NamedRerouteUsage()
    Emissive Color=LinearInterpolate(A=@n16, B=@n17.RGB, Alpha=@n15)
    Specular=LinearInterpolate(A=@n18, B=@n19, Alpha=@n20) {constB=0}
   ) {attributeSetTypes [#1, #2]}
   M=@n6.M
   Pre Skinned Local Position=@n6.M
   Distance Blend=@n6.M
   Fresnel=@n6.M
  ) {materialFunction=&MF_CharacterParent_2023_PostFX}
  NamedRerouteDeclaration(None=@n8.Specular) {name=InSpecular, nodeColor=<rgba 0 1 0.5547 1>}
  n21 = TextureSample() {samplerType=SAMPLERTYPE_Masks, texture=/CosmeticCompanions/Assets/Quadruped/PolarBlind_Vein/Texture/T_PolarBlind_TAMask}
  NamedRerouteDeclaration(None=@n21.B) {name=SkeletonHeadMask, nodeColor=<rgba 0.7812 1 0 1>}
  NamedRerouteDeclaration(None=@n21.R) {name=SkeletonAO, nodeColor=<rgba 0.8984 0 1 1>}
  NamedRerouteDeclaration(None=@n21.G) {name=SkeletonHeadAO, nodeColor=<rgba 0 1 0.5781 1>}
  NamedRerouteDeclaration(None=@n8.MaterialAttributes) {name=InMA, nodeColor=<rgba 1 0 0.3047 1>}
  n22 = TextureSample() {samplerType=SAMPLERTYPE_Masks, texture=/CosmeticCompanions/Assets/Quadruped/PolarBlind_Vein/Texture/T_PolarBlind_TAMask2}
  NamedRerouteDeclaration(None=Saturate(None=Multiply(A=@n22.G) {constB=5})) {name=TongueMask, nodeColor=<rgba 0 0.625 1 1>}
 ]
}
