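# Parent eye material for the PolarBlind Vein quadruped cosmetic companion, drawing two glowing spots on a near-black, rough surface.
# Two soft SphereMasks at fixed UV positions are summed, boosted and tinted by the Eye_Color parameter to drive Emissive, with no texture sampled in this graph.
# These attributes override the shared CharacterParent PreFX output, which then passes through the matching PostFX function.
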Material /CosmeticCompanions/Assets/Quadruped/PolarBlind_Vein/Material/M_PolarBlind_Vein_eye_Parent
graph {
 properties {
  bUseMaterialAttributes=true
  bUsedWithMeshDeformer=true
  bUsedWithMorphTargets=true
  bUsedWithSkeletalMesh=true
  pixelDepthOffsetMode=PDOM_Legacy
  usedShadingModels=Default Lit
 }
 nodes [
  n1 = MaterialFunctionCall() {materialFunction=&MF_CharacterParent_2023_PreFX}
  n2 = TextureCoordinate()
  n3 = ComponentMask(None=Constant4Vector() {constant=<rgba 0.91 0.085 0 1>}.RGB) {g=true, r=true}
  n4 = Constant() {r=0.038}
  n5 = Constant() {r=0.1}
  n6 = TextureCoordinate()
  n7 = ComponentMask(None=Constant4Vector() {constant=<rgba 0.91 0.25 0 1>}.RGB) {g=true, r=true}
  n8 = Multiply(A=Add(A=SphereMask(A=@n2, B=@n3, Radius=@n4, Hardness=@n5), B=SphereMask(A=@n6, B=@n7, Radius=@n4, Hardness=@n5))) {constB=3}
  n9 = VectorParameter() {group=StuffRasp, parameterName=Eye_Color, defaultValue=<rgba 1 1 1 1>}
  MaterialFunctionCall(
   Material=SetMaterialAttributes!(
    MaterialAttributes=GetMaterialAttributes(None=@n1.Result) {attributeGetTypes [#1]}.MaterialAttributes
    Emissive Color=Multiply(A=@n8, B=@n9.RGB)
    Base Color=Constant() {r=0.002}
    Roughness=Constant() {r=0.75}
   ) {attributeSetTypes [#1, #2, #3]}
   M=@n1.M
   Pre Skinned Local Position=@n1.M
   Distance Blend=@n1.M
   Fresnel=@n1.M
  ) {materialFunction=&MF_CharacterParent_2023_PostFX}
 ]
}
