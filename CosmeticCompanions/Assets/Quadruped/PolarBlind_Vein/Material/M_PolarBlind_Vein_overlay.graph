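# Translucent, two-sided Default Lit overlay material for the skeletal mesh of the PolarBlind Vein quadruped companion cosmetic.
# Character-parent PreFX attributes pass through overrides into PostFX, giving a rough, non-metallic surface with reduced specular.
# A tunable Fresnel sets dark greyscale base color.
# A second Fresnel drives opacity, zeroed on back faces through TwoSidedSign, which also feeds MF_SortTranslucency from the SortedOutline functions.
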Material /CosmeticCompanions/Assets/Quadruped/PolarBlind_Vein/Material/M_PolarBlind_Vein_overlay
graph {
 properties {
  allowTranslucentCustomDepthWrites=true
  bAutomaticallySetUsageInEditor=false
  bUseMaterialAttributes=true
  bUsedWithSkeletalMesh=true
  blendMode=BLEND_Translucent
  ditherOpacityMask=true
  translucencyLightingMode=TLM_SurfacePerPixelLighting
  twoSided=true
  usedShadingModels=Default Lit
 }
 nodes [
  n1 = GetMaterialAttributes(None=NamedRerouteUsage()) {attributeGetTypes [#1]}
  n2 = NamedRerouteUsage()
  n3 = SetMaterialAttributes!(
   MaterialAttributes=@n1.MaterialAttributes
   Base Color=NamedRerouteUsage()
   Metallic=Constant()
   Specular=Multiply(A=@n1.Specular, B=Constant() {r=0.15})
   Roughness=Constant() {r=0.8898}
   Opacity=@n2
   Normal
  ) {attributeSetTypes [
    #2
    #3
    #1
    #4
    #5
    #6
   ]}
  n4 = MaterialFunctionCall() {materialFunction=&MF_CharacterParent_2023_PreFX}
  MaterialFunctionCall(Material=@n3, M=@n4.M, Pre Skinned Local Position=@n4.M, Distance Blend=@n4.M, Fresnel=@n4.M) {materialFunction=&MF_CharacterParent_2023_PostFX}
  MaterialFunctionCall!(Opacity=@n2) {materialFunction=/Game/Global/Materials/MaterialGym/Materials/SortedOutline/MF_SortTranslucency}
  n5 = Constant()
  n6 = Constant() {r=0.98}
  n7 = Fresnel(ExponentIn=Constant() {r=0.03})
  n8 = Constant() {r=4.3522}
  n9 = Saturate(None=Multiply(A=Power(Base=@n7, Exp=@n8), B=Constant() {r=1.07}))
  n10 = TwoSidedSign()
  NamedRerouteDeclaration(None=LinearInterpolate(A=Constant(), B=LinearInterpolate(A=@n5, B=@n6, Alpha=@n9), Alpha=Saturate(None=@n10)) {constA=1, constB=0}) {name=Opacity, nodeColor=<rgba 0 1 0.25 1>}
  n11 = Fresnel(ExponentIn=ScalarParameter() {defaultValue=3, group=StuffRasp, parameterName=StuffRasp_FresnelExponentln})
  n12 = ScalarParameter() {defaultValue=1.5, parameterName=StuffRasp_FresnelPower}
  NamedRerouteDeclaration(None=LinearInterpolate(
    A=ScalarParameter() {defaultValue=0.01, group=StuffRasp, parameterName=StuffRasp_FresnelMin}
    B=ScalarParameter() {defaultValue=0.15, group=StuffRasp, parameterName=StuffRasp_FresnelMax}
    Alpha=Power(Base=@n11, Exp=@n12)
   )) {name=BaseColor, nodeColor=<rgba 1 0 0.0703 1>}
  NamedRerouteDeclaration(None=@n4.Result) {name=InMA, nodeColor=<rgba 0 0.3906 1 1>}
 ]
}
